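# One-shot mono MetaSound source for the DragonCart grenade pin-pull sound, using weapon foley attenuation and the gadgets/items sound class.
# On play, a random array get picks one of three pin-pull waves and triggers the wave player.
# Random floats vary the pitch shift and scale the output gain, though only the gain randomizer is retriggered by OnPlay.
# The wave player's finish trigger ends the one-shot.
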MetaSoundSource /DragonCartWeaponGameplay/Gameplay/DragonCart_Grenade/Audio/MSS_DragonCart_Grenade_PinPull
inputs [
 In Array {type=WaveAsset:Array}
 UE.Source.OnPlay {type=Trigger}
]
interfaces [UE.OutputFormat.Mono:1.0, UE.Source.OneShot:1.0, UE.Source:1.1]
outputs [
 "UE.OutputFormat.Mono.Audio:0" {type=Audio}
 UE.Source.OneShot.OnFinished {type=Trigger}
]
pages [Default {
  input_defaults {
   UE.Source.OnPlay=null
   In Array [
    /DragonCartWeaponGameplay/Gameplay/DragonCart_Grenade/Audio/Soundwaves/DragonCart_Grenade_PinPull_03
    /DragonCartWeaponGameplay/Gameplay/DragonCart_Grenade/Audio/Soundwaves/DragonCart_Grenade_PinPull_02
    /DragonCartWeaponGameplay/Gameplay/DragonCart_Grenade/Audio/Soundwaves/DragonCart_Grenade_PinPull_01
   ]
  }
  nodes [
   ArrayRandomGet_0 = Array.Random Get:WaveAsset:Array(In Array=^In Array, Next=^UE.Source.OnPlay) {also {defaults {Enable Shared State="<MetaSoundBuilderResult.SUCCEEDED: 0>", No Repeats=-1}}}
   WavePlayer_0 = UE.Wave Player:Mono(Play=@ArrayRandomGet_0.On Next, Wave Asset=@ArrayRandomGet_0.Value, Pitch Shift=UE.RandomFloat() {also {defaults {Min=-1}}}.Value) {also {version [1,1]}}
   Multiply_0 = UE.Multiply:Audio by Float(AdditionalOperands=UE.RandomFloat(Next=^UE.Source.OnPlay) {also {defaults {Min=0.7}}}.Value, PrimaryOperand=@WavePlayer_0.Out Mono)
  ]
  links [{dst="@UE.OutputFormat.Mono.Audio:0", src=Multiply_0.Out}, {dst="@UE.Source.OneShot.OnFinished", src=WavePlayer_0.On Finished}]
 }]
properties {
 assetImportData=&self:AssetImportData
 attenuationSettings=&AS_Fort_WeaponFoley
 maxDistance=2250
 sampleRate=48000
 soundClassObject=&Fort_Gadgets_Items
 volume=0.4
}
