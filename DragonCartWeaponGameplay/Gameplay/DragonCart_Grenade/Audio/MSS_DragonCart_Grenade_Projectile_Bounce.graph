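# One-shot mono MetaSound for the DragonCart grenade projectile's bounce, using its own attenuation asset and the ranged weapon projectile effects sound class.
# On play it picks a random bounce wave from an array of five variations and feeds it to a wave player.
# Chained random floats vary the pitch shift and multiply the output gain, so each bounce sounds different.
# The source reports finished when the wave player ends.
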MetaSoundSource /DragonCartWeaponGameplay/Gameplay/DragonCart_Grenade/Audio/MSS_DragonCart_Grenade_Projectile_Bounce
inputs [
 In Array {type=WaveAsset:Array}
 UE.Source.OnPlay {type=Trigger}
]
interfaces [UE.OutputFormat.Mono:1.0, UE.Source.OneShot:1.0, UE.Source:1.1]
outputs [
 "UE.OutputFormat.Mono.Audio:0" {type=Audio}
 UE.Source.OneShot.OnFinished {type=Trigger}
]
pages [Default {
  input_defaults {
   UE.Source.OnPlay=null
   In Array [
    /DragonCartWeaponGameplay/Gameplay/DragonCart_Grenade/Audio/Soundwaves/DragonCart_Grenade_Bounce_01
    /DragonCartWeaponGameplay/Gameplay/DragonCart_Grenade/Audio/Soundwaves/DragonCart_Grenade_Bounce_02
    /DragonCartWeaponGameplay/Gameplay/DragonCart_Grenade/Audio/Soundwaves/DragonCart_Grenade_Bounce_03
    /DragonCartWeaponGameplay/Gameplay/DragonCart_Grenade/Audio/Soundwaves/DragonCart_Grenade_Bounce_04
    /DragonCartWeaponGameplay/Gameplay/DragonCart_Grenade/Audio/Soundwaves/DragonCart_Grenade_Bounce_05
   ]
  }
  nodes [
   ArrayRandomGet_0 = Array.Random Get:WaveAsset:Array(In Array=^In Array, Next=^UE.Source.OnPlay)
   n1 = UE.RandomFloat(Next=@ArrayRandomGet_0.On Next) {also {defaults {Max=2, Min=-3}}}
   WavePlayer_0 = UE.Wave Player:Mono(Wave Asset=@ArrayRandomGet_0.Value, Play=@n1.On Next, Pitch Shift=@n1.Value) {also {version [1,1]}}
   Multiply_0 = UE.Multiply:Audio by Float(AdditionalOperands=UE.RandomFloat(Next=@n1.On Next) {also {defaults {Max=1.25, Min=0.75}}}.Value, PrimaryOperand=@WavePlayer_0.Out Mono)
  ]
  links [{dst="@UE.OutputFormat.Mono.Audio:0", src=Multiply_0.Out}, {dst="@UE.Source.OneShot.OnFinished", src=WavePlayer_0.On Finished}]
 }]
properties {
 assetImportData=&self:AssetImportData
 attenuationSettings=&/DragonCart_Grenade_Bounce_Att
 maxDistance=5500
 sampleRate=48000
 soundClassObject=&Fort_WeaponRanged_Projectile_Effects
 volume=1.45
}
