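# Looping fuse sound cue for the DragonCart grenade, layering three looping waves.
# A DragonCart-specific loop plays quietly through frag-grenade attenuation, alongside an M80 fuse and pre-explosion warning mix with volume fade-in, sound class override and Doppler.
# A PreExplo parameter drives continuous modulation of pitch and volume on both fuse and warning layers.
# Those two layers have separate attenuation overrides with HRTF spatialization, and only the warning enables occlusion.
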SoundCue /DragonCartWeaponGameplay/Gameplay/DragonCart_Grenade/Audio/SC_DragonCart_Grenade_Fuse_Loop
root=Mixer_3
nodes [
 {
  id=Attenuation_1
  children [WavePlayer_2]
  properties {
   attenuationSettings=/Game/Sounds/Fort_Weapons/Grenades/Frag/Grenade_Frag_InAir_Att
   attenuationOverrides {
    distanceAlgorithm=Logarithmic
    falloffDistance=1536
    spatializationAlgorithm=SPATIALIZATION_HRTF
    stereoSpread=0
    attenuationShapeExtents {x=768}
   }
  }
 }
 {
  id=Attenuation_6
  children [ModulatorContinuous_0]
  properties {
   attenuationSettings=&Fort_Weapon_Projectile_Medium_Mono_Athena_Att
   bOverrideAttenuation=true
   attenuationOverrides {
    bEnableReverbSend=false
    distanceAlgorithm=Logarithmic
    falloffDistance=256
    spatializationAlgorithm=SPATIALIZATION_HRTF
    stereoSpread=0
    attenuationShapeExtents {x=256}
   }
  }
 }
 {
  id=Attenuation_7
  children [ModulatorContinuous_1]
  properties {
   attenuationSettings=&Fort_Weapon_Projectile_Medium_Mono_Athena_Att
   bOverrideAttenuation=true
   attenuationOverrides {
    bEnableOcclusion=true
    bEnableReverbSend=false
    distanceAlgorithm=Logarithmic
    falloffDistance=0
    occlusionLowPassFilterFrequency=1000
    occlusionTraceChannel=ECC_GameTraceChannel11
    spatializationAlgorithm=SPATIALIZATION_HRTF
    stereoSpread=0
    attenuationShapeExtents {x=550}
   }
  }
 }
 {id=Doppler_0, children [Mixer_1], properties {dopplerIntensity=2}}
 {
  id=Enveloper_0
  children [SoundClass_0]
  properties {pitchCurve {editorCurveData {keys [{time=0}]}}, volumeCurve {editorCurveData {keys [<at 0 0>,<at 0.5 1>]}}}
 }
 {id=Mixer_1, children [Attenuation_6, Attenuation_7], properties {inputVolume [0.8,2.2]}}
 {id=Mixer_3, children [Attenuation_1, Enveloper_0], properties {inputVolume [0.2,1]}}
 {
  id=ModulatorContinuous_0
  children [Modulator_0]
  properties {
   pitchModulationParams {default=0, maxOutput=1.5, minOutput=1, parameterName=PreExplo}
   volumeModulationParams {default=0, maxOutput=1.2, minOutput=1, parameterName=PreExplo}
  }
 }
 {
  id=ModulatorContinuous_1
  children [WavePlayer_3]
  properties {
   pitchModulationParams {default=0, maxOutput=1.8, minOutput=1, parameterName=PreExplo}
   volumeModulationParams {default=0, minOutput=0.1, parameterName=PreExplo}
  }
 }
 {id=Modulator_0, children [WavePlayer_0]}
 {id=SoundClass_0, children [Doppler_0], properties {soundClassOverride=/Game/Sounds/FortSoundClassesAndModes/Fort_Explosives_Medium_PSM}}
 {id=WavePlayer_0, properties {bLooping=true, soundWaveAssetPtr=/Game/Sounds/Fort_GamePlay_Sounds/Explosives/fuse_m80_loop_shortened}}
 {
  id=WavePlayer_2
  properties {bLooping=true, soundWaveAssetPtr=/DragonCartWeaponGameplay/Gameplay/DragonCart_Grenade/Audio/Soundwaves/DragonCart_Grenade_Fuse_Loop}
 }
 {
  id=WavePlayer_3
  properties {bLooping=true, soundWaveAssetPtr=/Game/Sounds/Fort_GamePlay_Sounds/Explosives/GrenadeLauncher_Fuse_PreExploWarning_01}
 }
]
properties {
 maxDistance=6300
 soundClassObject=&Fort_WeaponRanged_Projectile_Effects
 volumeMultiplier=1
 attenuationOverrides {falloffDistance=5000, attenuationShapeExtents {x=1000}}
 concurrencyOverrides {bLimitToOwner=true, maxCount=3, resolutionRule=StopQuietest}
 concurrencySet [/Game/Sounds/Concurrency_Rules/Athena_Explosive_Fuse_Conc]
}
