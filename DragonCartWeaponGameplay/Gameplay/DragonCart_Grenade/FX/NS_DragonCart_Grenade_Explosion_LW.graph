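# Niagara system for the DragonCart grenade explosion, scaled by a user Global Scale.
# Burst emitters supply debris, flashes, flares with lights, and crack and scorch decals, alongside other explosion, smoke and streak emitters.
# GPU Niagara Fluids 2D grids simulate camera-facing smoke and fire from SourceParticles via a particle-read interface, plus a separate smoke trail grid.
# Distance culling, instance limits and platform checks bound the cost.
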NiagaraSystem /DragonCartWeaponGameplay/Gameplay/DragonCart_Grenade/FX/NS_DragonCart_Grenade_Explosion_LW
dependencies {
 usedDataInterfaces [
  &NiagaraDataInterfacePlatformSet
  &NiagaraDataInterfaceCamera
  &NiagaraDataInterfaceCurve
  &NiagaraDataInterfaceColorCurve
  /Script/Niagara.NiagaraDataInterfaceGrid2DCollection
  /Script/Niagara.NiagaraDataInterfaceRenderTarget2D
  &NiagaraDataInterfaceParticleRead
  /Script/Niagara.NiagaraDataInterfaceEmitterProperties
  &NiagaraDataInterfaceArrayFloat
  &NiagaraDataInterfaceVectorCurve
  &NiagaraDataInterfaceVector2DCurve
 ]
 usedDynamicInputs [
  /Niagara/DynamicInputs/Math/Min_Int
  &MakeIntFromFloat
  &MakeVector4_FromFloat
  &MakeCustomFloatFromBool
  &self:ScratchDynamicInput
  &self:ScratchDynamicInput_25
  &Vector2DFromCurve
 ]
 usedModules [
  &self:SystemSpawnScript.SystemScriptSource.SystemScriptGraph.NiagaraNodeAssignment_1.SetVariables_#1
  &self:SystemSpawnScript.SystemScriptSource.SystemScriptGraph.NiagaraNodeAssignment_3.SetVariables_#2
  &SystemState
  &self:SystemSpawnScript.SystemScriptSource.SystemScriptGraph.NiagaraNodeAssignment_0.SetVariables_#3
  &self:SystemSpawnScript.SystemScriptSource.SystemScriptGraph.NiagaraNodeAssignment_2.SetVariables_#4
 ]
}
emitters {
 Debris {
  simTarget=GPUComputeSim
  emitterUpdateScript [
   &EmitterState(
    Enable Distance Culling=true
    Life Cycle Mode=::NewEnumerator1
    Loop Behavior=::NewEnumerator1
    Loop Duration=0.125
    MaxDistance=4000
    Module.bUseMaxDistance=true
    Module.bUseMinDistance=false
    Scalability Mode=::NewEnumerator1
    UseLoopDelay=false
   )
   &SpawnBurst_Instantaneous(Spawn Count=&UniformRangedInt(Minimum=25, Maximum=30), Use Spawn Probability=false)
   &SpawnBurst_Instantaneous(Spawn Count=&UniformRangedInt(Minimum=50, Maximum=60), Spawn Group=1, Use Spawn Probability=false) {module=SpawnBurst_Instantaneous001}
  ]
  particleSpawnScript [
   /SetVariables(Particles.SizeAlpha:NiagaraFloat=&FloatFromCurve(
     FloatCurve=DataInterface() {propertyValues {Curve {keys [
         {arriveTangent=0.1408, interpMode=RCIM_Cubic, leaveTangent=0.1408, tangentMode=RCTM_User, time=0, value=0}
         {arriveTangent=2.7693, interpMode=RCIM_Cubic, leaveTangent=2.7693, tangentMode=RCTM_User, time=0.7, value=0.5197}
         {arriveTangent=-0.0596, interpMode=RCIM_Cubic, leaveTangent=-0.0596, tangentMode=RCTM_User, time=1, value=1}
        ]}}}
     CurveIndex=&RandomRangeFloat(Minimum=0, Maximum=1)
    )) {isSetParameters=true}
   /SetVariables(Particles.VisibilityTag=&SpawnGroup()) {isSetParameters=true}
   &InitializeParticle(
    Color=&MaskLinearColorBySpawnGroup(
     Masked Linear Color=<rgba 0 0 0 0.8>
     Passthrough Linear Color=&RandomRangeLinearColor(Minimum=<rgba 20 0.6667 0 1>, Maximum=<rgba 20 1.6667 0 1>)
    )
    Lifetime=&MaskFloatBySpawnGroup(
     Masked Float=&UniformRangedFloat(Minimum=2, Maximum=3.5)
     Passthrough Float=&UniformRangedFloat(Minimum=0.5, Maximum=1)
    )
    Sprite Rotation Mode=::NewEnumerator2
    Sprite Size=&Multiply_Vector2D_ByFloat(
     Float=$User.Global Scale
     Vector2D=&MaskVector2dBySpawnGroup(
      Masked Vector2D=&Vector2DFromFloat(Value=&Lerp_Float(A=2, B=10, Alpha=$Particles.SizeAlpha))
      Passthrough Vector2D=&Multiply_Vector2D_ByFloat(Float=&Lerp_Float(A=5, B=12, Alpha=$Particles.SizeAlpha), Vector2D=<0.4 2.5>)
      Spawn Group Mask=0
     )
    )
    Sprite Size Mode=::NewEnumerator4
    UsePositionOffset=true
    Write Material Random=true
   ) {module=InitializeParticle001}
   /Niagara/Modules/Update/SubUV/SubUVAnimation(SubUV Animation Mode=::NewEnumerator1)
   /SetVariables() {isSetParameters=true}
   &ShapeLocation(
    MaskSpawnGroup=false
    Offset=<0 0 -20>
    Offset Mode=::NewEnumerator0
    Sphere Distribution=::NewEnumerator1
    Sphere Radius=10
    U Position=&RandomRangeFloat(Minimum=0, Maximum=1)
    UseOwnerScale=true
    V Position=&MaskFloatBySpawnGroup(
     Masked Float=&RandomRangeFloat(Minimum=0.5, Maximum=1)
     Passthrough Float=&RandomRangeFloat(Minimum=0.9, Maximum=1)
    )
   )
   &AddVelocityFromPoint(
    Module.Use Falloff Distance=false
    Velocity Strength=&Multiply_Float(
     A=&Multiply_Float(
      A=&MaskFloatBySpawnGroup(
       Masked Float=&Lerp_Float(A=2550, B=450, Alpha=$Particles.SizeAlpha)
       Passthrough Float=&Lerp_Float(A=1750, B=1050, Alpha=$Particles.SizeAlpha)
      )
      B=&RandomRangeFloat(Minimum=1, Maximum=2.5)
     )
     B=$System.ExplosionScale
    )
   ) {enabled=false}
   &AddVelocity(
    Velocity=$Particles.ShapeLocation.ShapeNormal
    Velocity Speed Scale=&Multiply_Float(
     A=&Multiply_Float(
      A=&MaskFloatBySpawnGroup(
       Masked Float=&Lerp_Float(A=2550, B=450, Alpha=$Particles.SizeAlpha)
       Passthrough Float=&Lerp_Float(A=1750, B=1050, Alpha=$Particles.SizeAlpha)
      )
      B=&RandomRangeFloat(Minimum=1.25, Maximum=2)
     )
     B=$User.Global Scale
    )
   )
   &AddVelocityInCone(
    Cone Axis=<0 0 1>
    Module.Use Velocity Falloff On Cone Axis=false
    Velocity Strength=&UniformRangedFloat(Minimum=125, Maximum=1250)
   ) {enabled=false}
  ]
  particleUpdateScript [
   &ParticleState()
   &AccelerationForce(Acceleration=<0 0 -980>)
   &Drag(
    Drag=&MaskFloatBySpawnGroup(Masked Float=0, Passthrough Float=&Lerp_Float(A=3, B=1, Alpha=$Particles.SizeAlpha))
    Use Linear Drag=true
    Use Rotational Drag=true
   )
   &SolveForcesAndVelocity(Clamp Velocity=false, Limit Acceleration=false)
   &SpriteRotationRate(Rotation Rate=&AngleConversion(Angle:NiagaraFloat=0, Angle Input:ENiagara_AngleInput=::NewEnumerator0, Angle Output:ENiagara_AngleInput=::NewEnumerator1))
   &ScaleSpriteSize(Scale Factor=&Vector2DFromFloat(Value=&FloatFromCurve(FloatCurve=DataInterface() {propertyValues {Curve {keys [<at 0.6918 1>,<at 1 0>]}}})))
   &FN_Sprite_Attributes(
    Color=&MultiplyLinearColor(
     Linear Color=$Particles.Color
     Scale Factor=&MakeVector4_FromLinearColor(LinearColor=<rgba 0.5 0.5 0.5 0.01>)
    )
    Create Color Attribute=true
    Create SpriteSize Attribute=true
    SpriteSize=&Multiply_Vector2D(A=$Particles.SpriteSize, B=<10 3>)
    Write to Module Namespace=true
   )
  ]
  rendererClasses [&NiagaraSpriteRendererProperties]
  renderers [
   {rendererClass=&NiagaraSpriteRendererProperties}
   {rendererClass=&NiagaraSpriteRendererProperties, rendererIndex=1}
   {rendererClass=&NiagaraSpriteRendererProperties, rendererIndex=2}
  ]
 }
 DecalRenderer {
  emitterUpdateScript [
   &EmitterState()
   &SpawnBurst_Instantaneous(Use Spawn Probability=false, UseLoopCountLimit=false)
   /SetVariables(
    Emitter.OpacityMultCrack:NiagaraFloat=&RandomRangeFloat(Minimum=1, Maximum=1.5)
    Emitter.RandRotCracks:NiagaraFloat=&RandomRangeFloat(Minimum=0, Maximum=1)
   ) {isSetParameters=true}
  ]
  particleSpawnScript [&InitializeParticle(Lifetime=2, UsePositionOffset=false, Write Material Random=true)]
  particleUpdateScript [
   &ParticleState()
   /SetVariables(
    Particles.DecalFadeCrack:NiagaraFloat=&Lerp_Float(
     A=0
     B=1
     Alpha=&FloatFromCurve(FloatCurve=DataInterface() {propertyValues {Curve {keys [
          {arriveTangent=41.722, time=0, value=0}
          {interpMode=RCIM_Cubic, tangentMode=RCTM_User, time=0.02, value=1}
          {arriveTangent=-0.5041, interpMode=RCIM_Cubic, leaveTangent=-0.5041, tangentMode=RCTM_User, time=1, value=0}
         ]}}})
    )
    Particles.DecalFadeScorch:NiagaraFloat=&Lerp_Float(
     A=0
     B=1
     Alpha=&FloatFromCurve(FloatCurve=DataInterface() {propertyValues {Curve {keys [
          {arriveTangent=41.722, time=0, value=0.25}
          {interpMode=RCIM_Cubic, tangentMode=RCTM_User, time=0.02, value=1}
          {arriveTangent=-1.795, interpMode=RCIM_Cubic, leaveTangent=-1.795, tangentMode=RCTM_User, time=0.6, value=0.6}
          {arriveTangent=-3.611, arriveTangentWeight=0.112, interpMode=RCIM_Cubic, leaveTangent=-3.611, tangentMode=RCTM_User, time=1, value=0}
         ]}}})
    )
   ) {isSetParameters=true}
   /SetVariables(
    Particles.DecalSizeCrack:Vector3f=&Multiply_VectorByFloat(Vector=<300 1200 1200>, Float=$User.Global Scale)
    Particles.DecalSizeEmissive:Vector3f=&Multiply_VectorByFloat(Vector=<300 600 600>, Float=$User.Global Scale)
   ) {isSetParameters=true}
  ]
  rendererClasses [&NiagaraDecalRendererProperties]
  renderers [
   {rendererClass=&NiagaraDecalRendererProperties, rendererIndex=0}
   {rendererClass=&NiagaraDecalRendererProperties, rendererIndex=1}
  ]
 }
 Flash_Merged {
  emitterSpawnScript [/SetVariables(
    Emitter.CameraOffset:NiagaraFloat=&Multiply_Float(A=50, B=$User.Global Scale)
    Emitter.Color:LinearColor=<rgba 8 0 3.5772 0.8>
    Emitter.SpriteSize:Vector2f=&Multiply_Vector2D_ByFloat(Float=$User.Global Scale, Vector2D=<1500 20>)
   ) {isSetParameters=true}]
  emitterUpdateScript [
   &EmitterState(
    Enable Distance Culling=true
    Life Cycle Mode=::NewEnumerator1
    Loop Behavior=::NewEnumerator1
    Loop Duration=0.3247
    Max Distance Response=::KillAfterParticlesFinish
    MaxDistance=8500
    Module.bUseMaxDistance=true
    Module.bUseMinDistance=false
    Scalability Mode=::NewEnumerator1
    UseLoopDelay=true
   )
   &SpawnBurst_Instantaneous(Spawn Count=1, Use Spawn Probability=false)
   /SetVariables(
    Emitter.Color2:LinearColor=&MakeLinearColorFromVectorAndFloat(
     "Vector (RGB)"=&Lerp_Vector(
      A=&Multiply_VectorByFloat(
       Vector=<2 1 0>
       Float=&FNTimeOfDayScalar(Night Scale=2, Morning Scale=0.5, Day Scale=0.5, Evening Scale=0.5)
      )
      B=&Multiply_VectorByFloat(
       Vector=<16 5 1>
       Float=&FNTimeOfDayScalar(Night Scale=2, Morning Scale=0.5, Day Scale=0.5, Evening Scale=0.5)
      )
      Alpha=&FloatFromCurve(
       FloatCurve=DataInterface() {propertyValues {Curve {keys [
           {arriveTangent=-1.9247, interpMode=RCIM_Cubic, time=0, value=1}
           {arriveTangent=-1.8467, interpMode=RCIM_Cubic, leaveTangent=-1.8467, time=0.2339, value=0.25}
           {interpMode=RCIM_Cubic, tangentMode=RCTM_User, time=0.5415, value=0}
          ]}}}
       CurveIndex=$Emitter.NormalizedLoopAge
      )
     )
     "Float (Alpha)"=&FloatFromCurve(
      FloatCurve=DataInterface() {propertyValues {Curve {keys [{arriveTangent=-5.1417, interpMode=RCIM_Cubic, time=0.4976, value=1}, <at 0.7576 0 RCIM_Cubic>]}}}
      CurveIndex=$Emitter.NormalizedLoopAge
      Scale Curve=0.05
     )
    )
    Emitter.Scale:Vector3f=&VectorFromCurve(
     VectorCurve=DataInterface() {propertyValues {
       XCurve {keys [
         {arriveTangent=4.1785, interpMode=RCIM_Cubic, leaveTangent=4.1785, tangentMode=RCTM_User, time=0, value=0}
         {arriveTangent=0.9995, interpMode=RCIM_Cubic, leaveTangent=0.9995, tangentMode=RCTM_User, time=0.1905, value=0.6825}
         {arriveTangent=-0.0405, interpMode=RCIM_Cubic, leaveTangent=-0.0405, tangentMode=RCTM_User, time=0.6079, value=1.575}
        ]}
       YCurve {keys [
         {arriveTangent=4.1785, interpMode=RCIM_Cubic, leaveTangent=4.1785, tangentMode=RCTM_User, time=0, value=0}
         {arriveTangent=0.9995, interpMode=RCIM_Cubic, leaveTangent=0.9995, tangentMode=RCTM_User, time=0.1905, value=0.6825}
         {arriveTangent=-0.0405, interpMode=RCIM_Cubic, leaveTangent=-0.0405, tangentMode=RCTM_User, time=0.6079, value=1.575}
        ]}
       ZCurve {keys [
         {arriveTangent=4.1785, interpMode=RCIM_Cubic, leaveTangent=4.1785, tangentMode=RCTM_User, time=0, value=0}
         {arriveTangent=0.9995, interpMode=RCIM_Cubic, leaveTangent=0.9995, tangentMode=RCTM_User, time=0.1905, value=0.6825}
         {arriveTangent=-0.0405, interpMode=RCIM_Cubic, leaveTangent=-0.0405, tangentMode=RCTM_User, time=0.6079, value=1.575}
        ]}
      }}
     CurveIndex=$Emitter.NormalizedLoopAge
     Scale Curve=&Multiply_VectorByFloat(Vector=<2 2 2>, Float=$User.Global Scale)
    )
    Emitter.SpriteSize:Vector2f=&Multiply_Vector2D_ByFloat(
     Float=&FloatFromCurve(
      FloatCurve=DataInterface() {propertyValues {Curve {keys [
          {arriveTangent=9.901, interpMode=RCIM_Cubic, leaveTangent=9.901, tangentMode=RCTM_User, time=0.45, value=0}
          {arriveTangent=12.2238, interpMode=RCIM_Cubic, leaveTangent=12.2238, tangentMode=RCTM_User, time=0.7, value=3}
          <at 0.9 0>
         ]}}}
      CurveIndex=$Emitter.NormalizedLoopAge
     )
     Vector2D=$Emitter.Initial.SpriteSize
    )
   ) {isSetParameters=true}
  ]
  particleSpawnScript [
   &InitializeParticle(
    Color=<rgba 0.2479 1 0.2583 0.2>
    Lifetime=0.3
    Mesh Scale=&Multiply_VectorByFloat(
     Vector=&VectorFromFloat(Value=&Lerp_Float(A=1, B=0.9, Alpha=&MakeFloatFromInt(Int=&SpawnGroup())))
     Float=$User.Global Scale
    )
    Mesh Scale Mode=::NewEnumerator4
    Sprite Size Mode=::NewEnumerator3
    Uniform Sprite Size=&Multiply_Float(A=500, B=$User.Global Scale)
    UsePositionOffset=false
    Write Material Random=true
   ) {module=InitializeParticle001}
   &CameraOffset(Camera Offset Amount=&Multiply_Float(A=15, B=$User.Global Scale))
   /SetVariables(Particles.CamOfs2=80) {isSetParameters=true}
  ]
  particleUpdateScript [
   &ParticleState()
   &ScaleColor(
    Scale Alpha=&FloatFromCurve(FloatCurve=DataInterface() {propertyValues {Curve {keys [{arriveTangent=-3.5509, interpMode=RCIM_Cubic, leaveTangent=-3.5509, tangentMode=RCTM_User, time=0, value=1}, <at 1 0>]}}})
    Scale RGB=&VectorFromCurve(
     VectorCurve=DataInterface() {propertyValues {XCurve {keys [<at 0 1>,<at 1 0.1>]}, YCurve {keys [<at 0 1>,<at 1 0.1>]}, ZCurve {keys [<at 0 1>,<at 1 0.1>]}}}
     Scale Curve=&Multiply_VectorByFloat(Vector=<5 3 2>, Float=&FNTimeOfDayScalar(Night Scale=2, Morning Scale=1, Evening Scale=1))
    )
    ScaleA=true
    ScaleRGB=true
   )
   &SpriteRotationRate(
    Delta Time:NiagaraFloat=&FloatFromCurve(FloatCurve=DataInterface() {propertyValues {Curve {keys [
         {tangentMode=RCTM_User, time=0, value=1}
         {arriveTangent=-0.3174, interpMode=RCIM_Cubic, leaveTangent=-0.3174, tangentMode=RCTM_User, time=0.1832, value=1}
         {arriveTangent=0.0123, interpMode=RCIM_Cubic, leaveTangent=0.0123, tangentMode=RCTM_User, time=0.7355, value=0}
        ]}}})
    Rotation Rate=-5
   )
   &ScaleSpriteSize(
    Uniform Curve Scale=3
    Uniform Curve Sprite Scale=DataInterface() {propertyValues {Curve {keys [
        {arriveTangent=9.0725, interpMode=RCIM_Cubic, leaveTangent=9.0725, tangentMode=RCTM_User, time=0, value=0}
        {arriveTangent=-2.5249, interpMode=RCIM_Cubic, leaveTangent=-2.5249, tangentMode=RCTM_User, time=0.1003, value=0.9847}
        {arriveTangent=-0.0405, interpMode=RCIM_Cubic, leaveTangent=-0.0405, tangentMode=RCTM_User, time=0.6973, value=0}
       ]}}}
   )
   &ScaleMeshSize(Scale Factor=&VectorFromFloat(Value=&FloatFromCurve(
      FloatCurve=DataInterface() {propertyValues {Curve {keys [
          {arriveTangent=4.1785, interpMode=RCIM_Cubic, leaveTangent=4.1785, tangentMode=RCTM_User, time=0.0247, value=0}
          {arriveTangent=1.7203, interpMode=RCIM_Cubic, leaveTangent=1.7203, tangentMode=RCTM_User, time=0.2, value=0.5}
          <at 0.35 0.95>
          {arriveTangent=-0.0405, interpMode=RCIM_Cubic, leaveTangent=-0.0405, tangentMode=RCTM_User, time=0.55, value=0}
         ]}}}
      Scale Curve=&Multiply_Float(A=4, B=$User.Global Scale)
     )))
   /SetVariables(
    Particles.Color2=&ColorFromCurve(ColorCurve=DataInterface() {propertyValues {
       AlphaCurve {keys [<at 0 0.16 RCIM_Cubic>,<at 1 0 RCIM_Cubic>]}
       BlueCurve {keys [<at 0 0.5023 RCIM_Cubic>,<at 1 1 RCIM_Cubic>]}
       GreenCurve {keys [<at 0 5.4862 RCIM_Cubic>,<at 1 1 RCIM_Cubic>]}
       RedCurve {keys [<at 0 12 RCIM_Cubic>,<at 1 1 RCIM_Cubic>]}
      }})
    Particles.Color3:LinearColor=&MakeLinearColorFromVectorAndFloat(
     "Vector (RGB)"=&MakeVectorFromLinearColorRGB(Color=$User.Global Color)
     "Float (Alpha)"=&FloatFromCurve(FloatCurve=DataInterface() {propertyValues {Curve {keys [
          {interpMode=RCIM_Cubic, tangentMode=RCTM_User, time=-0.0046, value=0.1}
          {arriveTangent=-1.0036, interpMode=RCIM_Cubic, leaveTangent=-1.0036, tangentMode=RCTM_User, time=0.6, value=0.9}
          {arriveTangent=-2.4196, interpMode=RCIM_Cubic, leaveTangent=-2.4196, tangentMode=RCTM_User, time=1.0011, value=0.002}
         ]}}})
    )
   ) {isSetParameters=true}
  ]
  rendererClasses [&NiagaraSpriteRendererProperties, &NiagaraMeshRendererProperties]
  renderers [
   {rendererClass=&NiagaraSpriteRendererProperties}
   {rendererClass=&NiagaraMeshRendererProperties, rendererIndex=1}
   {rendererClass=&NiagaraMeshRendererProperties, rendererIndex=2}
   {rendererClass=&NiagaraSpriteRendererProperties, rendererIndex=3}
  ]
 }
 Grid2D_Gas_SmokeFire_Emitter {
  simTarget=GPUComputeSim
  emitterSpawnScript [
   /NiagaraFluids/Functions/General/Fluids_Gas_Manager(
    Density Bouyancy=-0.1
    Density Color Curve=DataInterface() {propertyValues {
      ExposedName=Curve Density
      bExposeCurve=true
      AlphaCurve {keys [<at 0 0 RCIM_Cubic>,<at 2.7618 1>]}
      BlueCurve {keys [<at 0.0245 0.045>,<at 4.4467 0.055>]}
      GreenCurve {keys [<at 0.0245 0.0307>,<at 4.4467 0.0375>]}
      RedCurve {keys [<at 0.0245 0.0266>,<at 4.4467 0.0325>]}
     }}
    Depth Fade=25
    Dissipation Temperature=5
    Dissipation Velocity=0.5
    EnvironmentColor:LinearColor=<rgba 0.1427 0.2031 0.3542 1>
    Pressure Relaxation=1
    Radius Scale=0.8
    Shadowing=0.25
    SunColor:LinearColor=<rgba 1.2417 1.53 2 1>
    Temperature Bouyancy=0.0025
    Temperature Color Curve=DataInterface() {propertyValues {
      ExposedName=Curve Temperature
      bExposeCurve=true
      AlphaCurve {keys [<at 0 0>,<at 0.4853 1>,<at 0.8256 1>]}
      BlueCurve {keys [<at 0.1597 0.0069>,<at 0.7312 0.0517>,<at 0.909 4.1667>]}
      GreenCurve {keys [<at 0.1597 0.0069>,<at 0.7312 0.729>,<at 0.909 12.4907>]}
      RedCurve {keys [<at 0.1597 0.0069>,<at 0.7312 8>,<at 0.909 50>]}
     }}
    Temperature Maximum=1280
    Temperature Scale=0.2
    Vorticity Confinement=0.1
   )
   /SetVariables(Emitter.SimGrid=DataInterface() {propertyValues {
      ClearBeforeNonIterationStage=true
      NumAttributes=0
      NumCellsMaxAxis=3
      NumCellsX=1
      NumCellsY=1
      OverrideBufferFormat=Float
      PreviewAttribute=None
      SetGridFromMaxAxis=false
      bOverrideFormat=false
      bPreviewGrid=false
      RenderTargetUserParameter {parameter {name=None, typeDefHandle {registeredTypeIndex=1}, varData []}}
      WorldBBoxSize=<1200 1200>
     }}) {isSetParameters=true}
   /SetVariables(Emitter.LightingGrid=DataInterface() {propertyValues {
      ClearBeforeNonIterationStage=true
      NumAttributes=3
      NumCellsMaxAxis=3
      NumCellsX=512
      NumCellsY=512
      OverrideBufferFormat=Float
      PreviewAttribute=None
      SetGridFromMaxAxis=false
      bOverrideFormat=false
      bPreviewGrid=false
      RenderTargetUserParameter {parameter {name=None, typeDefHandle {registeredTypeIndex=4}, varData []}}
      WorldBBoxSize=<100 100>
     }}) {isSetParameters=true}
   /SetVariables(Emitter.LightingGridDownsample=DataInterface() {propertyValues {
      ClearBeforeNonIterationStage=true
      NumAttributes=1
      NumCellsMaxAxis=3
      NumCellsX=64
      NumCellsY=64
      OverrideBufferFormat=Float
      PreviewAttribute=None
      SetGridFromMaxAxis=false
      bOverrideFormat=false
      bPreviewGrid=false
      RenderTargetUserParameter {parameter {name=None, typeDefHandle {registeredTypeIndex=4}, varData []}}
      WorldBBoxSize=<100 100>
     }}) {isSetParameters=true}
   /SetVariables() {isSetParameters=true}
   /SetVariables() {isSetParameters=true}
   /SetVariables() {isSetParameters=true}
   /NiagaraFluids/Modules/Grid2D/Grid2D_SetResolution(
    Grid:NiagaraDataInterfaceGrid2DCollection=$Emitter.SimGrid
    NumCellsMaxAxis=300
    Resolution Method:ENiagaraGrid2DResolution=::NewEnumerator1
    WorldGridExtents:Vector2f=&Multiply_Vector2D_ByFloat(Float=$User.Global Scale, Vector2D=<1280 1280>)
   )
   &Grid2D_SetResolution(
    Grid:NiagaraDataInterfaceGrid2DCollection=$Emitter.LightingGridDownsample
    OtherGrid:NiagaraDataInterfaceGrid2DCollection=$Emitter.SimGrid
    Resolution Method:ENiagaraGrid2DResolution=::NewEnumerator3
    Resolution Mult=0.2
    WorldGridExtents:Vector2f=$Emitter.Grid2D_SetResolution.WorldGridExtents
   ) {module=Grid2D_SetResolution003}
   &Grid2D_SetResolution(
    Grid:NiagaraDataInterfaceGrid2DCollection=$Emitter.LightingGrid
    OtherGrid:NiagaraDataInterfaceGrid2DCollection=$Emitter.SimGrid
    Resolution Method:ENiagaraGrid2DResolution=::NewEnumerator3
    WorldGridExtents:Vector2f=$Emitter.Grid2D_SetResolution.WorldGridExtents
   ) {module=Grid2D_SetResolution004}
   &Grid2D_SetResolution(
    Grid:NiagaraDataInterfaceGrid2DCollection=$Emitter.SourceGrid
    OtherGrid:NiagaraDataInterfaceGrid2DCollection=$Emitter.SimGrid
    Resolution Method:ENiagaraGrid2DResolution=::NewEnumerator3
    WorldGridExtents:Vector2f=$Emitter.Grid2D_SetResolution.WorldGridExtents
   ) {module=Grid2D_SetResolution005}
   /SetVariables(Emitter.g:Vector3f=<0 0 -980>) {isSetParameters=true}
   &Grid2D_SetResolution(
    Grid:NiagaraDataInterfaceGrid2DCollection=$Emitter.PressureGrid
    OtherGrid:NiagaraDataInterfaceGrid2DCollection=$Emitter.SimGrid
    Resolution Method:ENiagaraGrid2DResolution=::NewEnumerator3
    WorldGridExtents:Vector2f=$Emitter.Grid2D_SetResolution.WorldGridExtents
   ) {module=Grid2D_SetResolution001}
   &Grid2D_SetResolution(
    Grid:NiagaraDataInterfaceGrid2DCollection=$Emitter.TemporaryGrid
    OtherGrid:NiagaraDataInterfaceGrid2DCollection=$Emitter.SimGrid
    Resolution Method:ENiagaraGrid2DResolution=::NewEnumerator3
    WorldGridExtents:Vector2f=$Emitter.Grid2D_SetResolution.WorldGridExtents
   ) {module=Grid2D_SetResolution002}
   /SetVariables() {isSetParameters=true}
   /SetVariables(Emitter.dt:NiagaraFloat=$Engine.DeltaTime) {enabled=false, isSetParameters=true}
   /SetVariables(Emitter.ParticleSource=DataInterface() {propertyValues {EmitterBinding {bindingMode=Other, emitterName=SourceParticles}}}) {isSetParameters=true}
   /SetVariables() {isSetParameters=true}
   &self:Grid2D_Gas_SmokeFire_Emitter.NiagaraScratchPadContainer_2.SetLightingRT_0(LightingGrid:NiagaraDataInterfaceGrid2DCollection=$Emitter.LightingGrid, LightingRT:NiagaraDataInterfaceRenderTarget2D=$Emitter.LightingRT) {module=SetLightingRT_29}
   /SetVariables() {enabled=false, isSetParameters=true}
   /SetVariables(Emitter.density=1) {isSetParameters=true}
   /NiagaraFluids/Modules/Grid2D/Grid2D_UpdateSimAttributes(Grid:NiagaraDataInterfaceGrid2DCollection=$Emitter.SimGrid, WorldGridSize:Vector2f=$Emitter.Grid2D_SetResolution.WorldGridExtents) {module=UpdateSimAttributes}
   /SetVariables() {isSetParameters=true}
   &self:Grid2D_Gas_SmokeFire_Emitter.NiagaraScratchPadContainer_2.ScratchModule_02_1(New Grid2D Collection:NiagaraDataInterfaceGrid2DCollection=$Emitter.SimGrid) {module=ScratchModule_02001}
   /SetVariables(Emitter.PressureIterations=5) {isSetParameters=true}
   /NiagaraFluids/Modules/Grid2D/Grid2D_InitRT(Grid:NiagaraDataInterfaceGrid2DCollection=$Emitter.SimGrid, SimRT:NiagaraDataInterfaceRenderTarget2D=$Emitter.SimRT) {module=Grid2D_Gas_InitRT}
   &self:Grid2D_Gas_SmokeFire_Emitter.NiagaraScratchPadContainer_2.ScratchModule_0(
    Grid:NiagaraDataInterfaceGrid2DCollection=$Emitter.SimGrid
    PressureGrid:NiagaraDataInterfaceGrid2DCollection=$Emitter.PressureGrid
    TemporaryGrid:NiagaraDataInterfaceGrid2DCollection=$Emitter.TemporaryGrid
   ) {module=ScratchModule}
   /SetVariables(Emitter.GridsAllocated=true) {isSetParameters=true}
  ]
  emitterUpdateScript [
   &self:Emitter_Limitation_and_Culling__0(Max Num=3, Platform Set=DataInterface() {propertyValues {Platforms {qualityLevelMask=-17, cVarConditions [], deviceProfileStates []}}}) {module=Emitter_Limitation_and_Culling__0}
   &EmitterState(
    Enable Distance Culling=true
    Life Cycle Mode=::NewEnumerator1
    Loop Behavior=::NewEnumerator1
    Loop Duration=2.5
    Max Distance Response=::KillAfterParticlesFinish
    MaxDistance=7500
    Module.bUseMaxDistance=true
    Module.bUseMinDistance=false
    Scalability Mode=::NewEnumerator1
    UseLoopDelay=false
   )
   /SetVariables(Emitter.dt:NiagaraFloat=$Engine.DeltaTime) {isSetParameters=true}
   /NiagaraFluids/Modules/Grid2D/Grid2D_CreateUnitToWorldTransform(
    Camera:NiagaraDataInterfaceCamera=$Emitter.CameraQuery
    Camera Facing=true
    Offset:NiagaraPosition=$Engine.Owner.Position
    Rotation:Quat4f=$Engine.Owner.Rotation
    WorldGridExtents:Vector2f=$Emitter.Grid2D_SetResolution.WorldGridExtents
   )
   /SetVariables(Emitter.Position:NiagaraPosition=&ConvertVectorToPosition(
     Input Position=$Emitter.Grid2D_CreateUnitToWorldTransform.LocalToWorldTranslation
     Conversion Operation:ENiagara_LWCConvertVecToPos=::NewEnumerator0
    )) {isSetParameters=true}
   /NiagaraFluids/Modules/Grid2D/Grid2D_SetBounds(
    CameraFacing:NiagaraBool=$Emitter.Grid2D_CreateUnitToWorldTransform.CameraFacing
    EmitterProperties=DataInterface() {propertyValues {EmitterBinding {bindingMode=Self, emitterName=None}}}
    LocalPivot:Vector2f=$Emitter.Grid2D_CreateUnitToWorldTransform.LocalPivot
    Rotate for Camera Facing=true
    WorldGridExtents:Vector2f=$Emitter.Grid2D_SetResolution.WorldGridExtents
   )
   /SetVariables(Emitter.ExecutionState:ENiagaraExecutionState=$Emitter.ENiagaraExecutionState) {isSetParameters=true}
   &self:ScratchModule_02(PerformDeallocation:NiagaraBool=/Niagara/DynamicInputs/Scalability/IsEmitterComplete()) {module=ScratchModule_02002}
  ]
  rendererClasses [&NiagaraSpriteRendererProperties]
  renderers [
   {rendererClass=&NiagaraSpriteRendererProperties}
   {rendererClass=&NiagaraSpriteRendererProperties, rendererIndex=1}
  ]
 }
 Grid2D_Smoke_Trail_H {
  simTarget=GPUComputeSim
  emitterSpawnScript [
   /SetVariables(Emitter.Global Scale=1, Emitter.Global Smoke Color:LinearColor=<rgba 0.3229 0.3028 0.2791 1>) {isSetParameters=true}
   /SetVariables(Emitter.RT=DataInterface() {propertyValues {
      MipMapGeneration=Disabled
      MipMapGenerationType=Linear
      OverrideRenderTargetFilter=TF_Bilinear
      OverrideRenderTargetFormat=RTF_R8
      bInheritUserParameterSettings=false
      bOverrideFormat=true
      bPreviewRenderTarget=false
      RenderTargetUserParameter {parameter {name=None, typeDefHandle {registeredTypeIndex=4}, varData []}}
      Size=<96 96>
     }}) {isSetParameters=true}
   /SetVariables(Emitter.GridBufferA=DataInterface() {propertyValues {
      ClearBeforeNonIterationStage=true
      NumAttributes=0
      NumCellsMaxAxis=3
      NumCellsX=128
      NumCellsY=128
      OverrideBufferFormat=UnsignedNormalizedByte
      PreviewAttribute=Attribute At Index 0
      SetGridFromMaxAxis=false
      bOverrideFormat=true
      bPreviewGrid=false
      RenderTargetUserParameter {parameter {name=None, typeDefHandle {registeredTypeIndex=4}, varData []}}
      WorldBBoxSize=<100 100>
     }}) {isSetParameters=true}
   &Grid2D_SetResolution(Grid:NiagaraDataInterfaceGrid2DCollection=$Emitter.GridBufferA, NumCellsX=96, NumCellsY=96)
   /SetVariables(Emitter.CameraOffset=25, Emitter.Position:NiagaraPosition=<0 0 140>) {isSetParameters=true}
  ]
  emitterUpdateScript [
   &self:NiagaraEmitter_0.NiagaraScratchPadContainer_2.Emitter_Limitation_and_Culling__0(Max Num=2) {module=Emitter_Limitation_and_Culling_}
   &EmitterState(
    Enable Distance Culling=true
    Inactive Response=::NewEnumerator1
    Life Cycle Mode=::NewEnumerator1
    Loop Behavior=::NewEnumerator1
    Loop Delay=-0.3583
    Loop Duration=2.5
    Max Distance Response=::KillAfterParticlesFinish
    MaxDistance=9500
    Module.bUseMaxDistance=true
    Module.bUseMinDistance=false
    Scalability Mode=::NewEnumerator1
    UseLoopDelay=false
   )
   /SetVariables(Emitter.Fire Color:LinearColor=<rgba 15 2.4513 0 1>, Emitter.Outer Fire Color Hue Shift=0.97) {isSetParameters=true}
   /SetVariables(
    Emitter.Color:LinearColor=&ColorFromCurve(
     ColorCurve=DataInterface() {propertyValues {
       AlphaCurve {keys [<at 0.3374 1.25 RCIM_Cubic>,<at 0.9314 0 RCIM_Cubic>]}
       BlueCurve {keys [<at 0.0716 0.01 RCIM_Cubic>,<at 0.4288 0.035>,<at 1 0.0885>]}
       GreenCurve {keys [<at 0.0716 0.01 RCIM_Cubic>,<at 0.4288 0.035>,<at 1 0.0885>]}
       RedCurve {keys [<at 0.0716 0.01 RCIM_Cubic>,<at 0.4288 0.035>,<at 1 0.0885>]}
      }}
     CurveIndex=$Emitter.NormalizedLoopAge
    )
    Emitter.DynamicMaterialParameter:Vector4f=&MakeVector4(
     X=80
     Y=&FloatFromCurve(
      FloatCurve=DataInterface() {propertyValues {Curve {keys [
          {arriveTangent=1.1271, interpMode=RCIM_Cubic, leaveTangent=1.1271, tangentMode=RCTM_User, time=0, value=0.5}
          {arriveTangent=23.75, interpMode=RCIM_Cubic, leaveTangent=23.75, time=0.2288, value=3.0293}
          {arriveTangent=21.6445, interpMode=RCIM_Cubic, leaveTangent=21.6445, tangentMode=RCTM_User, time=0.4, value=10}
         ]}}}
      CurveIndex=$Emitter.NormalizedLoopAge
      Scale Curve=1.7
     )
     Z=&FloatFromCurve(
      FloatCurve=DataInterface() {propertyValues {Curve {keys [
          <at 0.04 5>
          <at 0.1 0.5>
          {arriveTangent=0.2332, interpMode=RCIM_Cubic, leaveTangent=0.2332, tangentMode=RCTM_User, time=0.2, value=0}
          <at 0.5 0>
         ]}}}
      CurveIndex=$Emitter.NormalizedLoopAge
      Scale Curve=32
     )
     W=0.01
    )
    Emitter.DynamicMaterialParameter1:Vector4f=&MakeVector4(
     X=256
     Y=&FN_PickFloatDependOnPlatform(TargetPlatform=DataInterface() {propertyValues {Platforms {qualityLevelMask=-28}}}, TargetPlatformActive=1.73, TargetPlatformNoActive=1.5)
     Z=&FloatFromCurve(FloatCurve=DataInterface() {propertyValues {Curve {keys [<at 1.25 0.4 RCIM_Cubic>,<at 1.6 0>]}}}, CurveIndex=$Emitter.Age)
     W=&FloatFromCurve(
      FloatCurve=DataInterface() {propertyValues {Curve {keys [<at 0 0.3>, {arriveTangent=-1.327, interpMode=RCIM_Cubic, leaveTangent=-1.327, tangentMode=RCTM_User, time=0.4, value=0.3}, <at 1 0>]}}}
      CurveIndex=$Emitter.NormalizedLoopAge
     )
    )
    Emitter.DynamicMaterialParameter2:Vector4f=&MakeVector4FromVectorAndFloat(
     "Vector (RGB)"=&MakeVectorFromLinearColorRGB(Color=$Emitter.Fire Color)
     "Float (Alpha)"=$Emitter.Outer Fire Color Hue Shift
    )
    Emitter.ExecutionState:ENiagaraExecutionState=$Emitter.ENiagaraExecutionState
    Emitter.SImAge:NiagaraFloat=&Add_Float(A=$Emitter.SImAge, B=$Engine.DeltaTime)
    Emitter.Smoke Color:LinearColor=&MultiplyLinearColor(
     Linear Color=&ColorFromCurve(
      ColorCurve=DataInterface() {propertyValues {
        AlphaCurve {keys [<at 0.4268 1.25 RCIM_Cubic>,<at 1 0 RCIM_Cubic>]}
        BlueCurve {keys [<at 0.0716 0.033 RCIM_Cubic>,<at 0.4288 0.117>,<at 1 0.297>]}
        GreenCurve {keys [<at 0.0716 0.033 RCIM_Cubic>,<at 0.4288 0.117>,<at 1 0.297>]}
        RedCurve {keys [<at 0.0716 0.033 RCIM_Cubic>,<at 0.4288 0.117>,<at 1 0.297>]}
       }}
      CurveIndex=$Emitter.NormalizedLoopAge
     )
     Scale Factor=&MakeVector4_FromLinearColor(LinearColor=$Emitter.Global Smoke Color)
    )
    Emitter.SpriteSize:Vector2f=<1000 1000>
   ) {isSetParameters=true}
   &self:NiagaraEmitter_0.NiagaraScratchPadContainer_2.Resizing_0() {module=Resizing}
  ]
  rendererClasses [&NiagaraSpriteRendererProperties]
  renderers [{rendererClass=&NiagaraSpriteRendererProperties}]
 }
 LW_Explosion_Main {rendererClasses [], renderers []}
 LW_GlowSphereMesh1 {bEnabled=false, rendererClasses [], renderers []}
 LW_Smoke_Ground {rendererClasses [], renderers []}
 LW_Streaks_Fat {rendererClasses [], renderers []}
 NE_Flare {
  emitterUpdateScript [
   &EmitterState(Life Cycle Mode=::NewEnumerator1, Loop Behavior=::NewEnumerator1, Loop Duration=0.4, UseLoopDelay=false)
   &SpawnBurst_Instantaneous(Spawn Count=1, Use Spawn Probability=false)
   &SpawnBurst_Instantaneous(Spawn Count=1, Spawn Group=1, Use Spawn Probability=false) {module=SpawnBurst_Instantaneous001}
   &SpawnBurst_Instantaneous(Spawn Count=1, Spawn Group=2, Use Spawn Probability=false) {module=SpawnBurst_Instantaneous002}
   &SpawnBurst_Instantaneous(Spawn Count=1, Spawn Group=3, Use Spawn Probability=false) {module=SpawnBurst_Instantaneous003}
  ]
  particleSpawnScript [
   &InitializeParticle(
    Color=<rgba 8 1.0667 0 1>
    Lifetime=&SelectFloatFromArray(
     Direct Array Index=&SpawnGroup()
     Float Selection Array=DataInterface() {propertyValues {FloatData [0.25,0.4,0.3,0.3]}}
    )
    Mass Mode=::NewEnumerator1
    Mesh Scale=&VectorFromFloat(Value=&Multiply_Float(A=&MaskFloatBySpawnGroup(Masked Float=10, Passthrough Float=3), B=$User.Global Scale))
    Mesh Scale Mode=::NewEnumerator4
    UsePositionOffset=true
    Write Material Random=true
   ) {module=InitializeParticle001}
   /SetVariables(Particles.VisibilityTag=&SpawnGroup()) {isSetParameters=true}
   &InitialMeshOrientation(Orientation Axis=<0 0 1>, Orientation Vector=<0 0 1>, Rotation=<0 0.25 0>, Use Orientation Vector=true, Use Rotation Vector=true)
   &InitialMeshOrientation(
    Orientation Axis=<0 0 0>
    Orientation Vector=<0 0 0>
    Rotation=&RandomRangeVector(Minimum=<0 0 0>, Maximum=<1 0 0>)
    Use Orientation Vector=false
    Use Rotation Vector=true
   ) {module=InitialMeshOrientation001}
   /SetVariables(Particles.BigFlareOrient:Quat4f=&OrientMeshToVector(Look At Direction=<0 0 0>, Reference Axis Vector=<1 0 0>)) {isSetParameters=true}
   /SetVariables(Particles.BigFlareDynmat:Vector4f=<0.8 0 0 0>) {isSetParameters=true}
  ]
  particleUpdateScript [
   &ParticleState()
   &ScaleMeshSize(Scale Factor=&VectorFromFloat(Value=&MaskFloatBySpawnGroup(
      Masked Float=&FloatFromCurve(FloatCurve=DataInterface() {propertyValues {Curve {keys [{arriveTangent=5.2157, interpMode=RCIM_Cubic, leaveTangent=5.2157, tangentMode=RCTM_User, time=0, value=0.3}, <at 0.4 1 RCIM_Cubic>]}}})
      Passthrough Float=&FloatFromCurve(FloatCurve=DataInterface() {propertyValues {Curve {keys [
           {interpMode=RCIM_Cubic, tangentMode=RCTM_Break, time=0, value=0.2}
           {arriveTangent=2.9365, interpMode=RCIM_Cubic, leaveTangent=2.9365, tangentMode=RCTM_User, time=0.1, value=1}
           {interpMode=RCIM_Cubic, tangentMode=RCTM_Break, time=1, value=2}
          ]}}})
     )))
   &ScaleColor(
    Scale Alpha=&MaskFloatBySpawnGroup(
     Masked Float=0
     Passthrough Float=&FloatFromCurve(FloatCurve=DataInterface() {propertyValues {Curve {keys [
          {interpMode=RCIM_Cubic, tangentMode=RCTM_User, time=0, value=1}
          {arriveTangent=-2.6886, interpMode=RCIM_Cubic, leaveTangent=-2.6886, tangentMode=RCTM_User, time=1, value=0}
         ]}}})
    )
    Scale RGB=&MaskVectorBySpawnGroup(
     Masked Vector=&VectorFromFloat(Value=&FloatFromCurve(FloatCurve=DataInterface() {propertyValues {Curve {keys [<at 0 0.7>,<at 0.7729 0.7>,<at 1 0>]}}}, Scale Curve=0.2))
     Passthrough Vector=&Multiply_Vector(
      A=&MakeVectorFromLinearColorRGB(Color=&ColorFromCurve(ColorCurve=DataInterface() {propertyValues {
          AlphaCurve {keys [<at 0 1>,<at 0.4696 1>,<at 1 0>]}
          BlueCurve {keys [<at 0 0.05>,<at 0.2681 0>,<at 0.4696 0.0698>,<at 1 0.037>]}
          GreenCurve {keys [<at 0 0.5707>,<at 0.2681 0.3605>,<at 0.4696 0.1719>,<at 1 0.0811>]}
          RedCurve {keys [<at 0 1>,<at 0.2681 1>,<at 0.4696 0.6979>,<at 1 0.3698>]}
         }}))
      B=&VectorFromFloat(Value=&FloatFromCurve(
        FloatCurve=DataInterface() {propertyValues {Curve {keys [
            {interpMode=RCIM_Cubic, tangentMode=RCTM_Break, time=0, value=10}
            {arriveTangent=-6.6072, interpMode=RCIM_Cubic, leaveTangent=-6.6072, tangentMode=RCTM_User, time=0.15, value=3}
            {interpMode=RCIM_Cubic, tangentMode=RCTM_Break, time=1, value=1}
           ]}}}
        Scale Curve=&Lerp_Float(A=0.1, B=1, Alpha=$System.CameraFadeAlpha)
       ))
     )
    )
    ScaleA=true
    ScaleRGB=true
   )
   &DynamicMaterialParameters(
    Index 0 Param 1=&FloatFromCurve(FloatCurve=DataInterface() {propertyValues {Curve {keys [{arriveTangent=-2.1015, interpMode=RCIM_Cubic, leaveTangent=-2.1015, tangentMode=RCTM_User, time=0, value=0.5}, <at 1 -0.15>]}}})
    Index 0 Param 2=&FloatFromCurve(FloatCurve=DataInterface() {propertyValues {Curve {keys [{arriveTangent=-0.3464, interpMode=RCIM_Cubic, leaveTangent=-0.3464, tangentMode=RCTM_User, time=0, value=1}, <at 1 0.9>]}}})
    Index 0 Param 3=&FloatFromCurve(FloatCurve=DataInterface() {propertyValues {Curve {keys [
         {interpMode=RCIM_Cubic, tangentMode=RCTM_Break, time=0, value=0}
         {arriveTangent=0.4478, interpMode=RCIM_Cubic, leaveTangent=0.4478, tangentMode=RCTM_User, time=0.15, value=1}
         {interpMode=RCIM_Cubic, tangentMode=RCTM_Break, time=1, value=1.2}
        ]}}})
    Index 3 Param 1=1
    Module.Param0WriteEnabled=true
    Module.Param0WriteEnabled3=true
    Module.Param1WriteEnabled=true
    Module.Param1WriteEnabled3=true
    Module.Param2WriteEnabled=true
    Module.Param2WriteEnabled3=true
    Module.Param3WriteEnabled=false
    Module.Param3WriteEnabled3=true
    Write Parameter Index 3=true
   )
   /SetVariables(Particles.BigFlareScale:Vector3f=&Multiply_VectorByFloat(Vector=$Particles.Scale, Float=0.2)) {isSetParameters=true}
   /SetVariables(Particles.BigFlareColor:LinearColor=&MultiplyLinearColor(
     Linear Color=<rgba 50 5 0 1>
     Scale Factor=&MakeFloat4FromVectorAndFloat(
      "Vector (RGB)"=&VectorFromFloat(Value=&Lerp_Float(A=0.05, B=1, Alpha=$System.CameraFadeAlpha))
      "Float (Alpha)"=&FloatFromCurve(FloatCurve=DataInterface() {propertyValues {Curve {keys [<at 0.4 1 RCIM_Cubic>,<at 1 0>]}}})
     )
    )) {isSetParameters=true}
   &FN_Light_Attributes(
    Color=$Particles.Color
    Create Color Attribute=true
    Create Exponent Attribute=true
    Create Position Attribute=true
    Create Radius Attribute=true
    Exponent=1
    Radius=&Multiply_Float(
     A=550
     B=&FloatFromCurve(
      FloatCurve=DataInterface() {propertyValues {Curve {keys [
          {arriveTangent=1.4112, interpMode=RCIM_Cubic, leaveTangent=1.4112, tangentMode=RCTM_User, time=0, value=0.2}
          {interpMode=RCIM_Cubic, tangentMode=RCTM_User, time=1, value=1}
         ]}}}
      Scale Curve=$User.Global Scale
     )
    )
    Scale Color=&FloatFromCurve(
     FloatCurve=DataInterface() {propertyValues {Curve {keys [
         {interpMode=RCIM_Cubic, tangentMode=RCTM_User, time=0, value=1}
         {arriveTangent=-3.1925, interpMode=RCIM_Cubic, leaveTangent=-3.1925, tangentMode=RCTM_User, time=1, value=0}
        ]}}}
     Scale Curve=100
    )
   )
   &FN_Sprite_Attributes(
    Camera Offset=150
    Color=&Multiply_LinearColorByFloat(Linear Color=$Particles.Color, Float=5)
    Create Camera Offset Attribute=true
    Create Color Attribute=true
    Create SpriteSize Attribute=true
    SpriteSize=&Multiply_Vector2D_ByFloat(
     Float=&Multiply_Float(
      A=&Multiply_Float(
       A=750
       B=&FloatFromCurve(FloatCurve=DataInterface() {propertyValues {Curve {keys [
            {arriveTangent=2.6957, interpMode=RCIM_Cubic, leaveTangent=2.6957, tangentMode=RCTM_User, time=0, value=0}
            {interpMode=RCIM_Cubic, tangentMode=RCTM_User, time=1, value=1}
           ]}}})
      )
      B=$User.Global Scale
     )
     Vector2D=<1 1>
    )
    Write to Module Namespace=true
   )
  ]
  rendererClasses [
   &NiagaraMeshRendererProperties
   &NiagaraLightRendererProperties
   &NiagaraSpriteRendererProperties
  ]
  renderers [
   {rendererClass=&NiagaraMeshRendererProperties, rendererIndex=0}
   {rendererClass=&NiagaraLightRendererProperties, rendererIndex=1}
   {rendererClass=&NiagaraMeshRendererProperties, rendererIndex=2}
   {rendererClass=&NiagaraMeshRendererProperties, rendererIndex=3}
   {rendererClass=&NiagaraSpriteRendererProperties, rendererIndex=4}
  ]
 }
 SourceParticles {
  simTarget=GPUComputeSim
  emitterUpdateScript [
   /NiagaraFN/Modules/Scalability/SignificantInstancesCull(
    Cull Response:ENiagaraExecutionStateManagement=::KillImmediately
    First Frame Only=false
    Num Instances=3
    Platform Set=DataInterface() {propertyValues {Platforms {qualityLevelMask=-17, cVarConditions [], deviceProfileStates []}}}
   )
   &EmitterState(
    Enable Distance Culling=true
    Life Cycle Mode=::NewEnumerator1
    Loop Behavior=::NewEnumerator1
    Loop Delay=0.015
    Loop Duration=0.2
    Max Distance Response=::KillAfterParticlesFinish
    MaxDistance=7500
    Module.bUseMaxDistance=true
    Module.bUseMinDistance=false
    Scalability Mode=::NewEnumerator1
    UseLoopDelay=true
   )
   &SpawnBurst_Instantaneous(
    Spawn Count=&RandomRangeInt(Minimum=5, Maximum=7, Evaluation Type=::NewEnumerator1)
    Use Spawn Probability=false
   )
   &SpawnRate(SpawnRate=55, Use Spawn Probability=false) {enabled=false}
   &Grid2D_CreateUnitToWorldTransform(
    Camera Facing=true
    Local Pivot:Vector2f=<0.5 0.1>
    Offset:NiagaraPosition=$Engine.Owner.Position
    Rotation:Quat4f=$Engine.Owner.Rotation
    WorldGridExtents:Vector2f=$Emitter.Grid2D_SetResolution.WorldGridExtents
   ) {enabled=false}
  ]
  particleSpawnScript [
   /Niagara/Modules/Spawn/Initialization/InitializeParticle(
    Color=<rgba 1111 13.6009 0 1>
    Lifetime=&RandomRangeFloat(Minimum=0.5, Maximum=1.1)
    Mass=&RandomRangeFloat(Minimum=0.75, Maximum=1)
    Module.Write Color=true
    Module.Write Lifetime=true
    Module.Write Mass=true
    Module.Write Material Random=true
    Module.Write Position=true
    Module.Write Scale=false
    Module.Write SpriteRotation=false
    Module.Write SpriteSize=true
    Sprite Size=&Vector2DFromFloat(Value=1)
   )
   &AddVelocityInCone(
    Cone Angle=80
    Cone Axis=<0 0 1>
    Cone Axis Coordinate Space=::World
    Module.Use Velocity Falloff On Cone Axis=true
    Velocity Distribution Along Cone Axis=0
    Velocity Falloff Away From Cone Axis=0
    Velocity Strength=&Multiply_Float(A=&RandomRangeFloat(Minimum=1075, Maximum=1756), B=$User.Global Scale)
   )
   &SphereLocation(
    MaskSpawnGroup=false
    Override Seed=false
    Sphere Radius=&Multiply_Float(
     A=&Lerp_Float(A=25, B=50, Alpha=&Sine(Normalized Angle=$Emitter.Age, Period=4, Scale=0.5, Bias=1))
     B=$User.Global Scale
    )
    Surface Only=false
    UseNonUniformScale=true
    UseOffset=true
    UseSphereOrientation=false
   ) {enabled=false}
   /Niagara/Modules/Update/Forces/CurlNoiseForce(
    Noise Frequency:NiagaraFloat=&Divide_Float(A=0.04, B=$User.Global Scale)
    Noise Strength=&Multiply_Float(A=250, B=$User.Global Scale)
    Pan Noise=true
    Pan Noise Field=<0.3 0.5 0.3>
   ) {enabled=false}
   &InheritVelocity(Module.Limit Speed=true, Velocity Limit=1500, Velocity Scale=<0 0 0>)
   &ApplyInitialForces()
   &TorusLocation(
    Align Torus With Axis=true
    Large Radius:NiagaraFloat=&Multiply_Float(A=60, B=$User.Global Scale)
    MaskSpawnGroup=false
    Torus Axis=<1 0 0>
    Torus Distribution Mode:ENiagaraTorusDistributionMode=::NewEnumerator1
    Torus Mode=::NewEnumerator2
    U Position:NiagaraFloat=&Subtract_Float(A=&ReturnNormalizedExecIndex(), B=2)
   )
   &AddVelocityFromPoint(
    Module.Use Falloff Distance=false
    Origin Offset=&Multiply_VectorByFloat(Vector=$Engine.Owner.SystemXAxis, Float=&RandomRangeFloat(Minimum=0, Maximum=-60))
    Origin Offset Coordinate Space=::Simulation
    Velocity Origin:Vector3f=&SimulationPosition()
    Velocity Strength=&Multiply_Float(A=&UniformRangedFloat(Minimum=5, Maximum=10), B=$User.Global Scale)
   ) {module=AddVelocityFromPoint001}
  ]
  particleUpdateScript [
   &ParticleState()
   &CurlNoiseForce(Noise Strength=300, Pan Noise=true, Pan Noise Field=<0.1 0.5 0.2>) {module=CurlNoiseForce001}
   &GravityForce(Gravity=&Multiply_VectorByFloat(Vector=<0 0 -1500>, Float=$User.Global Scale))
   &Drag(Drag=&Multiply_Float(A=1.4, B=$User.Global Scale), Use Linear Drag=true, Use Rotational Drag=false)
   &ScaleColor(
    Scale Alpha=&FloatFromCurve(FloatCurve=DataInterface() {propertyValues {Curve {keys [<at -0.0046 1.0113 RCIM_Cubic>,<at 1.0011 0.002 RCIM_Cubic>]}}})
    ScaleA=true
    ScaleRGB=true
   )
   &Collision(
    Age Colliding Particles=false
    Collision Enabled=&SetBoolByFloatComparison(A=$Particles.Age, B=0.4)
    Enable Rest State=false
    Friction=0.2
    Friction During a Bounce=0.3
    Kill Offscreen Particles=true
    Maximum time offscreen=99
    Module.EnableMaxCollisionCount=false
    Module.Randomize Collision Normal Vector=false
    Restitution=0.2
   ) {module=Collision001}
   &SolveForcesAndVelocity(Clamp Velocity=false, Limit Acceleration=false)
   /NiagaraFluids/Modules/Particles/Fluids_Gas_Source(
    NoiseIntensity=3
    NoiseScale=3
    Radius:NiagaraFloat=&FloatFromCurve(
     FloatCurve=DataInterface() {propertyValues {Curve {keys [<at 0 0>, {interpMode=RCIM_Cubic, leaveTangent=-2.282, tangentMode=RCTM_Break, time=0.0361, value=0.984}, <at 1.0022 0.3>]}}}
     Scale Curve=&Multiply_Float(A=&RandomRangeFloat(Minimum=30, Maximum=40), B=$User.Global Scale)
    )
    Temperature:NiagaraFloat=&FloatFromCurve(
     FloatCurve=DataInterface() {propertyValues {Curve {keys [<at 0 1>,<at 0.228 0.772>,<at 1 0>]}}}
     Scale Curve=&RandomRangeFloat(Minimum=500, Maximum=3000)
    )
    VelocityScale:NiagaraFloat=&RandomRangeFloat(Minimum=0.5, Maximum=1)
   )
   /SetVariables(Particles.SpriteSize=&Vector2DFromFloat(Value=&Multiply_Float(A=$Particles.fluids_source_radius, B=&Multiply_Float(A=1.4, B=1.1)))) {enabled=false, isSetParameters=true}
  ]
  rendererClasses [&NiagaraSpriteRendererProperties]
  renderers [{rendererClass=&NiagaraSpriteRendererProperties}]
 }
 Sparkles_and_Black_Chunks {
  simTarget=GPUComputeSim
  emitterUpdateScript [
   &EmitterState(
    Enable Distance Culling=true
    Life Cycle Mode=::NewEnumerator1
    Loop Behavior=::NewEnumerator1
    Loop Delay=0.0829
    Loop Duration=1.5729
    Max Distance Response=::KillAfterParticlesFinish
    MaxDistance=4000
    Module.bUseMaxDistance=true
    Module.bUseMinDistance=false
    Scalability Mode=::NewEnumerator1
    UseLoopDelay=true
   )
   &SpawnBurst_Instantaneous(Spawn Count=300, Use Spawn Probability=false, UseLoopCountLimit=false)
  ]
  particleSpawnScript [
   /SetVariables(Particles.VisibilityTag=$Emitter.SpawnGroup) {isSetParameters=true}
   /Niagara/Modules/Spawn/Initialization/V2/InitializeParticle(
    Color=&MaskLinearColorBySpawnGroup(Masked Linear Color=$User.Global Color, Passthrough Linear Color=<rgba 0 0 0 5>)
    Lifetime=&MaskFloatBySpawnGroup(
     Masked Float=&FloatFromCurve(
      FloatCurve=DataInterface() {propertyValues {Curve {keys [<at 0 0.25 RCIM_Constant>,<at 0.3 0.45>,<at 0.7 1 RCIM_Constant>,<at 0.85 1.5>,<at 1 2>]}}}
      CurveIndex=&RandomRangeFloat(Minimum=0, Maximum=1)
      Scale Curve=0.4
     )
     Passthrough Float=0.05
    )
    Sprite Rotation Mode=::NewEnumerator2
    Sprite Size=&MaskVector2dBySpawnGroup(
     Masked Vector2D=&RandomRangeVector2D(Minimum=<5 5>, Maximum=<6 9>)
     Passthrough Vector2D=&Multiply_Vector2D_ByFloat(Float=&RandomRangeFloat(Minimum=40, Maximum=50), Vector2D=<1 1>)
     Spawn Group Mask=0
    )
    Sprite Size Mode=::NewEnumerator4
    UsePositionOffset=false
    Write Material Random=true
   )
   /Niagara/Modules/Spawn/Location/SphereLocation(
    MaskSpawnGroup=false
    Module.Override Randomness=false
    Sphere Radius=&Multiply_Float(A=0, B=$User.Global Scale)
    Surface Only=false
   )
   &AddVelocity(
    Module.Use Falloff Distance=false
    Override Seed=false
    Velocity Mode=::NewEnumerator1
    Velocity Speed=&Multiply_Float(A=&RandomRangeFloat(Minimum=5000, Maximum=8000), B=$User.Global Scale)
   )
   &ApplyInitialForces(Apply Rotational Force to Rotational Velocity=false)
   /SetVariables(Particles.SpriteSize=&Multiply_Vector2D_ByFloat(Float=$User.Global Scale, Vector2D=$StackContext.SpriteSize)) {isSetParameters=true}
   &CameraOffset(Camera Offset Amount=&Multiply_Float(A=&MaskFloatBySpawnGroup(Masked Float=10, Passthrough Float=100), B=$User.Global Scale))
  ]
  particleUpdateScript [
   &ParticleState()
   &GravityForce(Gravity=&RandomRangeVector(Minimum=<0 0 -900>, Maximum=<0 0 -500>))
   &Drag(
    Drag=&MaskFloatBySpawnGroup(Masked Float=&RandomRangeFloat(Minimum=7, Maximum=15), Passthrough Float=20)
    Use Linear Drag=true
    Use Rotational Drag=false
   )
   &Collision(
    Age Colliding Particles=false
    Kill Offscreen Particles=true
    Module.Control Rest State Via Surface Normal=false
    Module.EnableMaxCollisionCount=false
    Module.Randomize Collision Normal Vector=false
   )
   /Niagara/Modules/Update/Forces/V2/CurlNoiseForce(Noise Frequency=15, Noise Quality / Cost=::NewEnumerator0, Noise Strength=500, Pan Noise=false)
   &SolveForcesAndVelocity(Clamp Velocity=false, Limit Acceleration=false)
   &ScaleColor(
    Scale Alpha=&FloatFromCurve(FloatCurve=DataInterface() {propertyValues {Curve {keys [
         {arriveTangent=26.7273, interpMode=RCIM_Cubic, leaveTangent=26.7273, tangentMode=RCTM_User, time=0.1, value=0.0144}
         {interpMode=RCIM_Cubic, tangentMode=RCTM_User, time=0.15, value=1}
         {interpMode=RCIM_Cubic, tangentMode=RCTM_User, time=0.85, value=1}
         {arriveTangent=-1.3701, interpMode=RCIM_Cubic, leaveTangent=-1.3701, tangentMode=RCTM_User, time=1.0011, value=0}
        ]}}})
    Scale RGB=&Lerp_Vector(
     A=<0 0 0>
     B=<12.5 4 0.8>
     Alpha=&FloatFromCurve(FloatCurve=DataInterface() {propertyValues {Curve {keys [
          {arriveTangent=15.3906, interpMode=RCIM_Cubic, leaveTangent=15.3906, tangentMode=RCTM_User, time=-0.0092, value=1}
          {interpMode=RCIM_Cubic, tangentMode=RCTM_User, time=0.0967, value=0.8}
          {interpMode=RCIM_Cubic, tangentMode=RCTM_User, time=0.45, value=0.8}
          <at 0.7 0 RCIM_Cubic>
         ]}}})
    )
    ScaleA=true
    ScaleRGB=true
   )
   &ScaleSpriteSize(Scale Factor=&Vector2DFromFloat(Value=&FloatFromCurve(FloatCurve=DataInterface() {propertyValues {Curve {keys [<at -0.0046 0>,<at 0.025 1>,<at 0.66 1>,<at 1.0011 0.5>]}}}, Scale Curve=2)))
   &ScaleSpriteSizeBySpeed(
    Max Scale Factor=&MaskVector2dBySpawnGroup(Masked Vector2D=<0.5 2.2>, Passthrough Vector2D=<1 1>, Spawn Group Mask=0)
    Min Scale Factor=&MaskVector2dBySpawnGroup(Masked Vector2D=<1 0.6>, Passthrough Vector2D=<1 1>, Spawn Group Mask=0)
    Velocity Threshold=1500
   )
  ]
  rendererClasses [&NiagaraSpriteRendererProperties]
  renderers [{rendererClass=&NiagaraSpriteRendererProperties}]
 }
 Sparks_GPU {
  simTarget=GPUComputeSim
  emitterSpawnScript [/SetVariables(
    Emitter.Global Color:LinearColor=$User.Global Color
    Emitter.Global Scale:NiagaraFloat=$User.Global Scale
    Emitter.RandomFloat:NiagaraFloat=&RandomRangeFloat(Minimum=-180, Maximum=180)
    Emitter.RandomInt:NiagaraInt32=&RandomRangeInt(Minimum=0, Maximum=100)
   ) {isSetParameters=true}]
  emitterUpdateScript [
   &EmitterState(
    Enable Distance Culling=true
    Life Cycle Mode=::NewEnumerator1
    Loop Behavior=::NewEnumerator1
    Loop Delay=0.01
    Loop Duration=0.5
    Max Distance Response=::KillAfterParticlesFinish
    MaxDistance:NiagaraFloat=&FN_PickFloatDependOnPlatform(TargetPlatform=DataInterface() {propertyValues {Platforms {qualityLevelMask=-24}}}, TargetPlatformActive=9000, TargetPlatformNoActive=6500)
    Module.bApplySpawnCountScaleByDistance=true
    Module.bUseMaxDistance=true
    Module.bUseMinDistance=false
    Scalability Mode=::NewEnumerator1
    Scale Spawn Count=true
    Spawn Count Scale By Distance=DataInterface() {propertyValues {Curve {keys [<at 0.3676 0.9657>,<at 0.9962 0.4982>]}}}
    UseLoopDelay=true
   )
   &SpawnBurst_Instantaneous(Spawn Count=200, Use Spawn Probability=false, UseLoopCountLimit=true)
   &SpawnBurst_Instantaneous(Spawn Count=50, Spawn Group=1, Use Spawn Probability=false, UseLoopCountLimit=true) {module=SpawnBurst_Instantaneous001}
  ]
  particleSpawnScript [
   &PartitionParticles(Partition Mode:ENiagara_PartitionMode_v2=::NewEnumerator7)
   &InitializeParticle(
    Color=$Emitter.Global Color
    Lifetime=&FloatFromCurve(
     FloatCurve=DataInterface() {propertyValues {Curve {keys [
         {arriveTangent=-9.9637, interpMode=RCIM_Cubic, leaveTangent=-9.9637, tangentMode=RCTM_User, time=0.0772, value=6.0113}
         {arriveTangent=-5.5088, interpMode=RCIM_Cubic, leaveTangent=-5.5088, time=0.2581, value=2.8753}
         <at 0.9935 0.9632 RCIM_Cubic>
        ]}}}
     CurveIndex=&RandomRangeFloat(Minimum=0, Maximum=1)
     Scale Curve=&MaskFloat(
      Mask Source=0
      Mask Parameter=?
      Mask=0
      Masked Float=&Multiply_Float(A=&RandomRangeFloat(Minimum=0.3, Maximum=0.5), B=$User.Global Scale)
      Passthrough Float=0.22
     )
    )
    Position Offset=&Multiply_VectorByFloat(Vector=<0 0 10>, Float=$User.Global Scale)
    Sprite Size=&Multiply_Vector2D_ByFloat(
     Float=&Multiply_Float(A=750, B=$User.Global Scale)
     Vector2D=&RandomRangeVector2D(Minimum=<0.019 0.032>, Maximum=<0.029 0.072>)
    )
    Sprite Size Mode=::NewEnumerator4
    UsePositionOffset=true
    Write Material Random=true
   )
   &AddVelocity(
    Velocity=&NormalizeVector(
     Vector To Normalize=&RandomRangeVector(
      Minimum=<-1 -1 0.1>
      Maximum=<1 1 1>
      Randomness Mode=::NewEnumerator1
      Random Seed=&Add_Integer(
       A=&RandomRangeInt(Minimum=0, Maximum=6)
       B=&Multiply_Int(
        A=$Emitter.RandomInt
        B=/Niagara/DynamicInputs/Helpers/MaskInt(
         Mask Source:NiagaraInt32=0
         Mask Parameter:NiagaraInt32=?
         Mask:NiagaraInt32=0
         Masked Int:NiagaraInt32=99
         Passthrough Int:NiagaraInt32=&RandomRangeInt(Minimum=0, Maximum=200)
        )
       )
      )
      Fixed Random Seed=true
     )
     Safe Normalize=true
     Safe Fallback Threshold=0
     Safe Fallback Vector=<0 0 0.001>
    )
    Velocity Speed Scale=&FloatFromCurve(
     FloatCurve=DataInterface() {propertyValues {Curve {keys [
         {arriveTangent=-2.0184, interpMode=RCIM_Cubic, leaveTangent=-2.0184, tangentMode=RCTM_User, time=0, value=1}
         {arriveTangent=-1.2785, interpMode=RCIM_Cubic, leaveTangent=-1.2785, tangentMode=RCTM_User, time=1, value=0.3534}
        ]}}}
     CurveIndex=&RandomRangeFloat(Minimum=0, Maximum=1)
     Scale Curve=&MaskFloat(Mask Source=0, Mask Parameter=?, Mask=0, Masked Float=4000, Passthrough Float=5000)
    )
   )
   &AddVelocity(
    Velocity=&RandomVector(Vector Scale=1)
    Velocity Speed Scale=&FloatFromCurve(
     FloatCurve=DataInterface() {propertyValues {Curve {keys [
         {arriveTangent=-2.7658, interpMode=RCIM_Cubic, leaveTangent=-2.7658, tangentMode=RCTM_User, time=0, value=1}
         {arriveTangent=-0.635, interpMode=RCIM_Cubic, leaveTangent=-0.635, tangentMode=RCTM_User, time=0.1761, value=0.3439}
         <at 1 0 RCIM_Cubic>
        ]}}}
     CurveIndex=&RandomRangeFloat(Minimum=0, Maximum=1)
     Scale Curve=400
    )
   ) {enabled=false, module=AddVelocity001}
   /SetVariables(Particles.Velocity=&Multiply_VectorByFloat(Vector=$Particles.Velocity, Float=$User.Global Scale)) {isSetParameters=true}
   &CameraOffset(Camera Offset Amount=50)
  ]
  particleUpdateScript [
   &ParticleState()
   &CurlNoiseForce(
    Bias Noise=false
    Noise Frequency=1
    Noise Strength=&Multiply_Float(
     A=1000
     B=&FloatFromCurve(FloatCurve=DataInterface() {propertyValues {Curve {
         defaultValue=0
         keys [
          <at 0.0046 1 RCIM_Cubic>
          {arriveTangent=-0.1399, interpMode=RCIM_Cubic, leaveTangent=-0.1399, tangentMode=RCTM_User, time=0.0908, value=0.171}
          {arriveTangent=0.062, interpMode=RCIM_Cubic, leaveTangent=0.062, tangentMode=RCTM_User, time=0.9, value=0}
         ]
        }}})
    )
    Pan Noise=false
   )
   &Drag(
    Drag=&MaskFloat(
     Mask Source=0
     Mask Parameter=?
     Mask=0
     Masked Float=&FloatFromCurve(
      FloatCurve=DataInterface() {propertyValues {Curve {keys [<at 0.1 4>,<at 0.5007 1>]}}}
      CurveIndex=&Add_Float(A=$Particles.Age, B=&RandomRangeFloat(Minimum=-0.12, Maximum=0))
      Scale Curve=&UniformRangedFloat(Minimum=0.3, Maximum=1)
     )
     Passthrough Float=&FloatFromCurve(FloatCurve=DataInterface() {propertyValues {Curve {keys [<at 0.4 1>,<at 1 0.2>]}}}, Scale Curve=5)
    )
    Use Linear Drag=true
    Use Rotational Drag=true
   )
   &GravityForce(Gravity=&MaskVector(
     Mask Source=0
     Mask Parameter=?
     Mask=0
     Masked Vector=&Multiply_VectorByFloat(Vector=&RandomRangeVector(Minimum=<0 0 -800>, Maximum=<0 0 -650>), Float=$Emitter.Global Scale)
     Passthrough Vector=<0 0 -100>
    ))
   &PointForce(
    Force Falloff Distance=3000
    Force Strength=&MaskFloat(Mask Source=0, Mask Parameter=?, Mask=0, Masked Float=100, Passthrough Float=0)
    Linear Force Vector:Vector3f=<1 0 -25>
    Module.Use Falloff=true
    Module.UseLinearForce=true
   )
   &ScaleColor(
    Scale Alpha=&FloatFromCurve(FloatCurve=DataInterface() {propertyValues {Curve {keys [
         {arriveTangent=26.7273, interpMode=RCIM_Cubic, leaveTangent=26.7273, tangentMode=RCTM_User, time=-0.0092, value=0.0144}
         {interpMode=RCIM_Cubic, tangentMode=RCTM_User, time=0.1, value=1}
         {interpMode=RCIM_Cubic, tangentMode=RCTM_User, time=0.85, value=1}
         {arriveTangent=-1.3701, interpMode=RCIM_Cubic, leaveTangent=-1.3701, tangentMode=RCTM_User, time=1.0011, value=0}
        ]}}})
    Scale RGB=&Lerp_Vector(
     A=<0 0 0>
     B=<12.5 4 0.8>
     Alpha=&FloatFromCurve(FloatCurve=DataInterface() {propertyValues {Curve {keys [
          {arriveTangent=15.3906, interpMode=RCIM_Cubic, leaveTangent=15.3906, tangentMode=RCTM_User, time=-0.0092, value=1}
          {interpMode=RCIM_Cubic, tangentMode=RCTM_User, time=0.0967, value=0.8}
          {interpMode=RCIM_Cubic, tangentMode=RCTM_User, time=0.45, value=0.8}
          <at 0.7 0 RCIM_Cubic>
         ]}}})
    )
    ScaleA=true
    ScaleRGB=true
   )
   &CurlNoiseForce(
    Noise Frequency=&Divide_Float(A=15, B=$Emitter.Global Scale)
    Noise Quality / Cost=::NewEnumerator0
    Noise Strength=&Multiply_Float(A=500, B=$Emitter.Global Scale)
    Pan Noise=false
   ) {enabled=false, module=CurlNoiseForce001}
   &SolveForcesAndVelocity(Clamp Velocity=false, Limit Acceleration=false)
   &ScaleSpriteSize(Scale Factor=&Vector2DFromFloat(Value=&FloatFromCurve(
      FloatCurve=DataInterface() {propertyValues {Curve {keys [<at 0 1>,<at 1 0>]}}}
      Scale Curve=&RandomRangeFloat(Minimum=0.2, Maximum=0.7, Evaluation Type=::NewEnumerator1)
     )))
   &ScaleSpriteSizeBySpeed(
    Max Scale Factor=&Lerp_Vector2(
     A=<1 0.8>
     B=<0.75 4>
     Alpha=&FloatFromCurve(FloatCurve=DataInterface() {propertyValues {Curve {keys [
          {arriveTangent=-3.7663, interpMode=RCIM_Cubic, leaveTangent=-3.7663, tangentMode=RCTM_User, time=0, value=1}
          {interpMode=RCIM_Cubic, tangentMode=RCTM_User, time=0.4, value=0}
         ]}}})
    )
    Velocity Threshold:NiagaraFloat=&Multiply_Float(A=1000, B=1)
   )
  ]
  rendererClasses [&NiagaraSpriteRendererProperties]
  renderers [{rendererClass=&NiagaraSpriteRendererProperties}]
 }
}
summary {
 systemName=NS_DragonCart_Grenade_Explosion_LW
 emitters [
  {emitterName=SourceParticles, simTarget=GPUComputeSim, rendererClasses [&NiagaraSpriteRendererProperties]}
  {emitterName=Grid2D_Gas_SmokeFire_Emitter, simTarget=GPUComputeSim, rendererClasses [&NiagaraSpriteRendererProperties]}
  {bEnabled=false, emitterName=LW_GlowSphereMesh1, rendererClasses []}
  {emitterName=LW_Streaks_Fat, rendererClasses []}
  {emitterName=LW_Smoke_Ground, rendererClasses []}
  {emitterName=LW_Explosion_Main, rendererClasses []}
  {emitterName=Grid2D_Smoke_Trail_H, simTarget=GPUComputeSim, rendererClasses [&NiagaraSpriteRendererProperties]}
  {emitterName=DecalRenderer, rendererClasses [&NiagaraDecalRendererProperties]}
  {
   emitterName=NE_Flare
   rendererClasses [
    &NiagaraMeshRendererProperties
    &NiagaraLightRendererProperties
    &NiagaraSpriteRendererProperties
   ]
  }
  {emitterName=Debris, simTarget=GPUComputeSim, rendererClasses [&NiagaraSpriteRendererProperties]}
  {emitterName=Flash_Merged, rendererClasses [&NiagaraSpriteRendererProperties, &NiagaraMeshRendererProperties]}
  {emitterName=Sparkles_and_Black_Chunks, simTarget=GPUComputeSim, rendererClasses [&NiagaraSpriteRendererProperties]}
  {emitterName=Sparks_GPU, simTarget=GPUComputeSim, rendererClasses [&NiagaraSpriteRendererProperties]}
 ]
 userVariables [
  User.Additional Fire Color {type=LinearColor, description="", defaultValue {struct=LinearColor, value=<rgba 2 1 1 1>}}
  User.Explosion Smoke Color {type=LinearColor, description="", defaultValue {struct=LinearColor, value=<rgba 0.1462 0.1462 0.225 1>}}
  User.ExplosionScale {type=NiagaraFloat, defaultValue=0.5, description=""}
  User.Fire Intensity {type=NiagaraFloat, defaultValue=1, description=""}
  User.Global Color {type=LinearColor, description="", defaultValue {struct=LinearColor, value=<rgba 7 0.2333 0 1>}}
  User.Global Scale {type=NiagaraFloat, defaultValue=0.5, description=""}
  User.Heat Dissipation {type=NiagaraFloat, defaultValue=0.5, description=""}
  User.PhysicalSurface {type=NiagaraInt32, defaultValue=30}
  User.Spawn Decal {type=NiagaraBool, defaultValue=true}
  User.Use Iconic Blast {type=NiagaraBool, defaultValue=true}
 ]
}
