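# Niagara system for the DragonCart grenade that, going by its name, shows a brief heat screen effect.
# A single emitter uses an instantaneous burst to spawn a very large, short-lived sprite, apparently to cover the view.
# Spawn sets a random material value and writes dynamic material parameters for the sprite's material.
# Over its lifetime, a color curve shifts the tint from warm to neutral white while fading alpha to zero.
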NiagaraSystem /DragonCartWeaponGameplay/Gameplay/DragonCart_Grenade/FX/NS_DragonCart_Grenade_HeatScreenFX
dependencies {
 usedDataInterfaces [&NiagaraDataInterfaceColorCurve]
 usedDynamicInputs [/Game/Effects/Niagara/DynamicInputs/State/FN_IsExecutionStateActive]
 usedModules [&SystemState, &Timeline]
}
emitters {Minimal001 {
  emitterUpdateScript [
   &EmitterState(
    Inactive Response=::NewEnumerator1
    Life Cycle Mode=::NewEnumerator1
    Loop Behavior=::NewEnumerator1
    Loop Duration Mode=::NewEnumerator1
    UseLoopDelay=false
   )
   &SpawnBurst_Instantaneous(Use Spawn Probability=false, UseLoopCountLimit=false)
  ]
  particleSpawnScript [
   &InitializeParticle(Lifetime=0.3, Sprite Size Mode=::NewEnumerator3, Uniform Sprite Size=10000, UsePositionOffset=false, Write Material Random=true)
   &DynamicMaterialParameters(
    Index 0 Param 1=1
    Index 0 Param 2=2
    Module.Index0Float0Write=true
    Module.Index0Float1Write=true
    Module.Index0Float2Write=true
    Module.Index0Float3Write=true
   )
  ]
  particleUpdateScript [
   &ParticleState()
   &ScaleColor(
    Linear Color Curve=DataInterface() {propertyValues {
      AlphaCurve {keys [<at 0 1>,<at 1 0>]}
      BlueCurve {keys [<at 0 0.8125>,<at 1 1>]}
      GreenCurve {keys [<at 0 0.9>,<at 1 1>]}
      RedCurve {keys [<at 0 1.25>,<at 1 1>]}
     }}
    Scale Mode=2
   )
  ]
  rendererClasses [&NiagaraSpriteRendererProperties]
  renderers [{rendererClass=&NiagaraSpriteRendererProperties}]
 }}
summary {
 systemName=NS_DragonCart_Grenade_HeatScreenFX
 emitters [{emitterName=Minimal001, rendererClasses [&NiagaraSpriteRendererProperties]}]
 userVariables []
}
