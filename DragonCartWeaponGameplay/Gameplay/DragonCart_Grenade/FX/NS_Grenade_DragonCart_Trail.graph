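# Niagara ribbon trail effect for the DragonCart grenade, with no user parameters.
# A single emitter spawns particles by distance moved, so trail density follows movement rather than time.
# Over each particle's life, a curve drives a widening ScaledRibbonWidth, while color curves shift from warm red through muted tones to dark as alpha fades.
# Constant upward acceleration makes the trail drift up, suggesting rising smoke.
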NiagaraSystem /DragonCartWeaponGameplay/Gameplay/DragonCart_Grenade/FX/NS_Grenade_DragonCart_Trail
dependencies {
 usedDataInterfaces [&NiagaraDataInterfaceCurve, &NiagaraDataInterfaceVectorCurve]
 usedDynamicInputs []
 usedModules [&SystemState]
}
emitters {Ribbon {
  emitterSpawnScript [/SetVariables(Emitter.Static bool=true) {isSetParameters=true}]
  emitterUpdateScript [
   &EmitterState(Life Cycle Mode=::NewEnumerator1, Loop Duration=1, UseLoopDelay=false)
   &SpawnPerUnit(
    Module.Use Max Movement Threshold=true
    Module.Use Movement Tolerance=true
    Movement Tolerance=0.1
    Spawn Spacing:NiagaraFloat=&Divide_Float(A=1, B=&Divide_Float(A=1, B=50))
    Use Spawn Probability=false
   )
  ]
  particleSpawnScript [
   &InitializeRibbon(Ribbon Width=20, UsePositionOffset=false, Write Material Random=true)
   &DynamicMaterialParameters(Module.Index0Param0Write=true, Module.Index0Param1Write=false, Module.Index0Param2Write=false, Module.Index0Param3Write=false)
  ]
  particleUpdateScript [
   /SetVariables(Particles.ScaledRibbonWidth:NiagaraFloat=&Multiply_Float(
     A=$Particles.RibbonWidth
     B=&FloatFromCurve(
      FloatCurve=DataInterface() {propertyValues {Curve {keys [{arriveTangent=9.5068, interpMode=RCIM_Cubic, time=0, value=0.2}, {arriveTangent=0.0733, leaveTangent=0.0733, time=1, value=1}]}}}
      Scale Curve=7
     )
    )) {isSetParameters=true}
   &ParticleState()
   &Color(Color=&MakeLinearColorFromVectorAndFloat(
     "Vector (RGB)"=&VectorFromCurve(
      VectorCurve=DataInterface() {propertyValues {
        XCurve {keys [
          <at 0.0333 3>
          {arriveTangent=-8.5017, interpMode=RCIM_Cubic, leaveTangent=-8.5017, time=0.18, value=0.5}
          {arriveTangent=-1.2134, interpMode=RCIM_Cubic, leaveTangent=-1.2134, time=0.3273, value=0.5}
          <at 0.5921 0 RCIM_Cubic>
         ]}
        YCurve {keys [
          <at 0.0333 0.6>
          {arriveTangent=-0.3401, interpMode=RCIM_Cubic, leaveTangent=-0.3401, time=0.18, value=0.5}
          {arriveTangent=-1.2134, interpMode=RCIM_Cubic, leaveTangent=-1.2134, time=0.3273, value=0.5}
          <at 0.5921 0 RCIM_Cubic>
         ]}
        ZCurve {keys [
          <at 0.1 0.6707>
          <at 0.2 0.6707 RCIM_Cubic>
          {arriveTangent=-1.25, interpMode=RCIM_Cubic, leaveTangent=-1.25, time=0.3, value=0.6707}
          <at 0.6 0.1707 RCIM_Cubic>
         ]}
       }}
      Scale Curve=&VectorFromFloat(Value=0.3)
     )
     "Float (Alpha)"=&FloatFromCurve(FloatCurve=DataInterface() {propertyValues {Curve {keys [{arriveTangent=-3.0258, interpMode=RCIM_Cubic, leaveTangent=-3.0258, tangentMode=RCTM_User, time=0.1016, value=1}, <at 1 0>]}}})
    ))
   &AccelerationForce(Acceleration=<0 0 250>)
   &SolveForcesAndVelocity(Clamp Velocity=false, Limit Acceleration=false)
  ]
  rendererClasses [&NiagaraRibbonRendererProperties]
  renderers [{rendererClass=&NiagaraRibbonRendererProperties}]
 }}
summary {
 systemName=NS_Grenade_DragonCart_Trail
 emitters [{emitterName=Ribbon, rendererClasses [&NiagaraRibbonRendererProperties]}]
 userVariables []
}
